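# Physics asset for the Wurst vehicle body cosmetic, giving its mesh a simplified collision shape.
# A single box on the root bone stands in for the body, raised above the bone origin and shifted slightly along X.
# It has no other bodies and no constraints, so it defines no jointed or ragdoll behavior.
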PhysicsAsset /VehicleCosmetics/Bodies/Wurst/Meshes/Body_Wurst_PhysicsDM
bodies {root [{shapeType=Box, extentX=444, extentY=216, extentZ=132, shapeName=root_box, center=<5.8474 0 105.2541>}]}
constraints []
